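; Victory Crowns HUD component that manages the post-match UI scenes for earning or winning with a crown.
; When UI is ready, it listens on the global event router for the local player's crown-won and Royal Royale-won events.
; A crown win adds the Earned scene; a Royal Royale win adds the Won scene and removes Earned.
; EndPlay stops both listeners and removes both scenes.
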(event EventOnReadyToShowUI
  (bind _returnvalue (GameplayEventRouter|GetGlobalEventRouter))
  (bind _returnvalue_2 (CallFunction|GetOwningLocalPlayerController))
  (bind _asynctaskproxy (GameplayEventRouter|StartListeningToEvent _returnvalue "/Script/VictoryCrownsRuntime.FortEvent_HasWonCrownInMatch" _returnvalue_2)
    (:then
      (Variables|Default|SetOnHasWonCrownInMatchEvent _asynctaskproxy)
      (bind _asynctaskproxy_2 (GameplayEventRouter|StartListeningToEvent _returnvalue "/Script/VictoryCrownsRuntime.FortEvent_HasWonRoyalRoyale" _returnvalue_2)
        (:then
          (Variables|Default|SetOnHasWonRoyalRoyaleEvent _asynctaskproxy_2))
        (:OnEventReceived
          (CallFunction|AddScene (Variables|Default|GetPostMatchSceneWonVictoryCrown))
          (CallFunction|RemoveScene (Variables|Default|GetPostMatchSceneEarnedVictoryCrown)))))
    (:OnEventReceived
      (CallFunction|AddScene (Variables|Default|GetPostMatchSceneEarnedVictoryCrown)))))

(event EventEndPlay
  (params (EndPlayReason :type EEndPlayReason%20Enum :default "Destroyed"))
  (GameplayEventRouter|StopListeningtoEvent (Variables|Default|GetOnHasWonCrownInMatchEvent))
  (CallFunction|RemoveScene (Variables|Default|GetPostMatchSceneEarnedVictoryCrown))
  (GameplayEventRouter|StopListeningtoEvent (Variables|Default|GetOnHasWonRoyalRoyaleEvent))
  (CallFunction|RemoveScene (Variables|Default|GetPostMatchSceneWonVictoryCrown)))
